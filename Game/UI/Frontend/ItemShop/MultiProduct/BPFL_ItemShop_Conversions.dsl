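; Blueprint function library of display helpers for the Item Shop's offer tiles, grids and purchase UI.
; Image ruleset selection tests full-tile override, offer type, vehicle, smart cards, bundle, then item type.
; It reads DataTable rows named by tile size, with smart-card column or Jam Track suffixes where relevant.
; Other helpers derive text, colors, alignment, column counts, preview eligibility, withdrawal-dialog checks and config-driven purchase button style variants.
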(fn CanShowOfferDescription
  (params (Description :type Text) (OfferType :type EFortItemShopOfferType%20Enum :default "LevelBundle"))
  (return (not (or (Utilities|Text|TextIsEmpty (Tile|GetOfferDescriptionText Description OfferType)) (Utilities|Select :Index OfferType :None true :LevelBundle true :BattlePass false :ProgressiveCosmetic false :Crew false :Special true :Regular false :MainItemType false :MainItemTypeAndBonus false :Pack false :OutfitPack false :PackAndBonus false :Vehicle false :VehicleAndBonus false :Bundle false :LEGOKit false)))))

(fn CompareItemDefinitions
  (params (ItemA :type Fort%20Item%20VM%20Object%20Reference) (ItemB :type Fort%20Item%20VM%20Object%20Reference))
  (Variables|GetItemA
    (:then
      (Variables|GetItemB
        (:then
          (return (== (Variables|/Script/FortniteUI.FortItemVM|GetItemDefinition (Variables|GetItemA)) (Variables|/Script/FortniteUI.FortItemVM|GetItemDefinition (Variables|GetItemB)))))
        (:else
          _)))
    (:else
      _))
  (return false))

(fn FindItemByDefinition
  (params (ItemArray :type Array%20of%20Fort%20Item%20VM%20Object%20References) (Item :type Fort%20Item%20VM%20Object%20Reference))
  (for (_array_element _array_index) ItemArray
    (if (Utility|CompareItemDefinitions _array_element Item)
      (return _array_index)))
  (return -1))

(fn GetContentGatingInterfacesFromOffer
  (params (OfferVM :type Fort%20Item%20Shop%20Offer%20VM%20Object%20Reference))
  (for _array_element (Variables|/Script/FortniteUI.FortItemShopOfferVM|GetGrantedItems OfferVM)
    (bind _ascosmetic_item_gating_interface (Utilities|Casting|CastToCosmeticItemGatingInterface (Variables|/Script/FortniteUI.FortItemVM|GetItemDefinition _array_element)))
    (Utilities|Array|Add (LocalVariables|GetOutInterfaces) _ascosmetic_item_gating_interface))
  (return (LocalVariables|GetOutInterfaces)))

(fn GetAlignmentFromStringMetadata
  (params (StringMetadata :type Map%20of%20Names%20to%20Strings))
  (bind (_value _returnvalue) (Utilities|Map|Find StringMetadata "Alignment"))
  (if _returnvalue
    (switch string _value
      (:Default
        _)
      (:Left
        _)
      (:Right
        (return "HAlign_Right" true))
      (:Center
        (return "HAlign_Center" true)))
    (return "HAlign_Left" true)
    (else
      (return "HAlign_Left" false))))

(fn GetArtistNameFromGrantedItems
  (params (GrantedItems :type Array%20of%20Fort%20Item%20VM%20Object%20References))
  (if (== (Utilities|Array|Length GrantedItems) 1)
    (return (select (Utilities|Enum|Equal(Enum) (Class|FortItemFunctionLibrary|GetItemType (Variables|/Script/FortniteUI.FortItemVM|GetItemDefinition (Utilities|Array|Get(acopy) GrantedItems))) "SparksSong") (Class|FortItemDetailsVM|GetItemShortDescription (Variables|/Script/FortniteUI.FortItemVM|GetItemDetails (Utilities|Array|Get(acopy) GrantedItems))) ""))
    (else
      (return ""))))

(fn GetColumnCountFromOffers
  (params (OfferVMs :type Array%20of%20Fort%20Item%20Shop%20Offer%20VM%20Object%20References) (UseWidePreview :type Boolean))
  (if UseWidePreview
    (return 5)
    (else
      (for _array_element (Variables|GetOfferVMs)
        (bind (_gridspan_x _gridspan_y) (Utilities|Struct|BreakIntPoint (TileGrid|GetGridSpanfromTileSize (Variables|/Script/FortniteUI.FortItemShopOfferVM|GetTileSize _array_element))))
        (if (== _gridspan_y 1)
          (if (LocalVariables|GetIsPreviousColumnFull)
            (Math|Integer|IncrementInt (LocalVariables|GetCounter)
              (:%20%20
                (LocalVariables|SetIsPreviousColumnFull false) ; Start new column, but another small tile could fit underneath.
              ))
            (else
              (LocalVariables|SetIsPreviousColumnFull true) ; Fits underneath previous small in the same column.\nDo not increment column count.
            ))
          (else
            (LocalVariables|SetCounter (+ (LocalVariables|GetCounter) _gridspan_x))
            (LocalVariables|SetIsPreviousColumnFull true) ; Fills entire column(s)
          )))
      (return (Math|Integer|Clamp(Integer) (LocalVariables|GetCounter) 4 5)))))

(fn GetFinalTextColor
  (params (DisplayData :type Fort%20Item%20Shop%20Offer%20Meta%20Display%20Data%20Structure) (OfferType :type EFortItemShopOfferType%20Enum :default "LevelBundle") (IsFilterActive :type Boolean))
  (bind (_r _g _b) (Math|Color|BreakColor (Utilities|Struct|BreakFortItemShopOfferMetaDisplayData (Class|FortItemShopBlueprintLibrary|GetMetaDisplayDataforCurrentViewMode DisplayData IsFilterActive))))
  (return (Utilities|Select :Index (Variables|GetOfferType) :None (Utilities|Struct|MakeLinearColor _r _g _b 1.0) :LevelBundle (Math|Color|MakeColor _r _g _b 0.4) :BattlePass (Math|Color|MakeColor _r _g _b 0.4) :ProgressiveCosmetic (Utilities|Struct|MakeLinearColor _r _g _b 1.0) :Crew (Math|Color|MakeColor _r _g _b 0.4) :Special (Math|Color|MakeColor _r _g _b 0.4) :Regular (Utilities|Struct|MakeLinearColor _r _g _b 1.0) :MainItemType (Utilities|Struct|MakeLinearColor _r _g _b 1.0) :MainItemTypeAndBonus (Utilities|Struct|MakeLinearColor _r _g _b 1.0) :Pack (Utilities|Struct|MakeLinearColor _r _g _b 1.0) :OutfitPack (Utilities|Struct|MakeLinearColor _r _g _b 1.0) :PackAndBonus (Utilities|Struct|MakeLinearColor _r _g _b 1.0) :Vehicle (Utilities|Struct|MakeLinearColor _r _g _b 1.0) :VehicleAndBonus (Utilities|Struct|MakeLinearColor _r _g _b 1.0) :Bundle (Utilities|Struct|MakeLinearColor _r _g _b 1.0) :LEGOKit (Utilities|Struct|MakeLinearColor _r _g _b 1.0))))

(fn GetGridSpanFromTileSize
  (params (TileSize :type EItemShopTileSize%20Enum :default "Small"))
  (return (Utilities|Select :Index (Variables|GetTileSize) :Size_1_x_1 (Utilities|Struct|MakeIntPoint 1 1) :Size_2_x_1 (Utilities|Struct|MakeIntPoint 2 1) :Size_3_x_1 (Utilities|Struct|MakeIntPoint 3 1) :Size_4_x_1 (Utilities|Struct|MakeIntPoint 4 1) :Size_5_x_1 (Utilities|Struct|MakeIntPoint 5 1) :Size_1_x_2 (Utilities|Struct|MakeIntPoint 1 1) :Size_2_x_2 (Utilities|Struct|MakeIntPoint 2 1) :Size_3_x_2 (Utilities|Struct|MakeIntPoint 3 1) :Size_4_x_2 (Utilities|Struct|MakeIntPoint 4 1) :Size_5_x_2 (Utilities|Struct|MakeIntPoint 5 1))))

(fn GetImageDisplayRuleset
  (params (OfferVM :type Fort%20Item%20Shop%20Offer%20VM%20Object%20Reference) (ItemShopVM :type Fort%20Item%20Shop%20VM%20Object%20Reference) (ItemsToShow :type Array%20of%20Fort%20Item%20VM%20Object%20References) (IsFilteringActive :type Boolean))
  (bind _offervm (Variables|GetOfferVM))
  (bind _displayedprimarygranteditems (Variables|/Script/FortniteUI.FortItemShopOfferVM|GetDisplayedPrimaryGrantedItems _offervm))
  (if (Variables|/Script/FortniteUI.FortItemShopOfferVM|GetbUsingFullTileOverridePresentations (Variables|GetOfferVM))
    ;[0 0 0 0.5] ruleset for 4x1 full tile override texture
    (bind _rulesetvalues (Tile|GetRuleforRuleset "/FortUILibrary/Systems/ItemShopTile/Rulesets/DT_ItemShop_ImageRuleset_FullTileSpan.DT_ItemShop_ImageRuleset_FullTileSpan" (Variables|/Script/FortniteUI.FortItemShopOfferVM|GetTileSize (Variables|GetOfferVM))))
    (return _rulesetvalues "FullTileSpan")
    (elif (Utilities|Select :Index (Variables|/Script/FortniteUI.FortItemShopOfferVM|GetOfferType (Variables|GetOfferVM)) :None false :LevelBundle true :BattlePass true :ProgressiveCosmetic true :Crew true :Special true :Regular false :MainItemType false :MainItemTypeAndBonus false :Pack false :OutfitPack false :PackAndBonus false :Vehicle false :VehicleAndBonus false :Bundle false :LEGOKit false)
      ;[0 0 0] ruleset for RMTs, Passes and FN Crew
      (bind _rulesetvalues_2 (Tile|GetRuleforRuleset "/FortUILibrary/Systems/ItemShopTile/Rulesets/DT_ItemShop_ImageRuleset_Basic.DT_ItemShop_ImageRuleset_Basic" (Variables|/Script/FortniteUI.FortItemShopOfferVM|GetTileSize (Variables|GetOfferVM))))
      (return _rulesetvalues_2 "Basic")
      (elif (Tile|ShouldUseVehicleRuleset (Variables|GetOfferVM))
        ;[0 0 0] ruleset for cars (keep this check before the smart cards check, vehicles have their own smart card rules)
        (bind (_rulevalues _rulename) (Tile|GetRuleforRulesetwithSmartCards (Variables|GetOfferVM) "/FortUILibrary/Systems/ItemShopTile/Rulesets/DT_ItemShop_ImageRuleset_Vehicle.DT_ItemShop_ImageRuleset_Vehicle" (Variables|GetItemsToShow) (Variables|GetIsFilteringActive)))
        (return _rulevalues (Utilities|String|Append "Vehicle " _rulename))
        (elif (Utilities|Array|IsNotEmpty (Variables|GetItemsToShow))
          ;[0 0 0] ruleset for smart cards
          (bind (_rulevalues_2 _rulename_2) (Tile|GetRuleforRulesetwithSmartCards (Variables|GetOfferVM) "/FortUILibrary/Systems/ItemShopTile/Rulesets/DT_ItemShop_ImageRuleset_SmartCards.DT_ItemShop_ImageRuleset_SmartCards" (Variables|GetItemsToShow) (Variables|GetIsFilteringActive)))
          (return _rulevalues_2 (Utilities|String|Append "Smart Cards " _rulename_2))
          (elif (Utilities|Enum|Equal(Enum) (Variables|/Script/FortniteUI.FortItemShopOfferVM|GetOfferType (Variables|GetOfferVM)) "Bundle")
            ;[0 0 0] ruleset for bundle offers
            (bind (_rulevalues_3 _rulename_3) (Tile|GetRuleforBundleOffer (Variables|GetOfferVM)))
            (return _rulevalues_3 (Utilities|String|Append "Bundle " _rulename_3))
            (elif (Utilities|Array|IsValidIndex _displayedprimarygranteditems)
              ;[0 0 0] ruleset for most item types
              (bind (_rulesetvalues_3 _ruleset_name) (Tile|GetRuleforItemType (Utilities|Array|Get(acopy) _displayedprimarygranteditems) (Variables|/Script/FortniteUI.FortItemShopOfferVM|GetTileSize (Variables|GetOfferVM)) (Variables|GetItemShopVM)))
              (return _rulesetvalues_3 _ruleset_name))))))))

(fn GetMaxSmartCardGridSize
  (params (TileSize :type EItemShopTileSize%20Enum :default "Size_1_x_1") (UsingVehicleRuleset :type Boolean) (IsFilteringActive :type Boolean))
  (return (select (Variables|GetIsFilteringActive) (Utilities|Select :Index (Variables|GetTileSize) :Size_1_x_1 3 :Size_2_x_1 6 :Size_3_x_1 9 :Size_4_x_1 13 :Size_5_x_1 13 :Size_1_x_2 3 :Size_2_x_2 6 :Size_3_x_2 9 :Size_4_x_2 13 :Size_5_x_2 13) (select (Variables|GetUsingVehicleRuleset) (Utilities|Select :Index (Variables|GetTileSize) :Size_1_x_1 2 :Size_2_x_1 2 :Size_3_x_1 3 :Size_4_x_1 4 :Size_5_x_1 4 :Size_1_x_2 2 :Size_2_x_2 2 :Size_3_x_2 3 :Size_4_x_2 4 :Size_5_x_2 4) (Utilities|Select :Index (Variables|GetTileSize) :Size_1_x_1 2 :Size_2_x_1 2 :Size_3_x_1 3 :Size_4_x_1 4 :Size_5_x_1 4 :Size_1_x_2 2 :Size_2_x_2 2 :Size_3_x_2 3 :Size_4_x_2 4 :Size_5_x_2 4))) (select (Variables|GetIsFilteringActive) 1 (select (Variables|GetUsingVehicleRuleset) 4 2))))

(fn GetNumVisibleSmartCardColumns
  (params (ItemsToShow :type Array%20of%20Fort%20Item%20VM%20Object%20References) (OfferVM :type Fort%20Item%20Shop%20Offer%20VM%20Object%20Reference) (GridSizeOverride :type Int%20Point%20Structure) (IsFilteringActive :type Boolean))
  (bind (_gridsize_x _gridsize_y) (Utilities|Struct|BreakIntPoint (Tile|GetMaxSmartCardGridSize (Variables|/Script/FortniteUI.FortItemShopOfferVM|GetTileSize (Variables|GetOfferVM)) (Tile|ShouldUseVehicleRuleset (Variables|GetOfferVM)) (Variables|GetIsFilteringActive))))
  (bind (_gridsizeoverride_x _gridsizeoverride_y) (Utilities|Struct|BreakIntPoint (Variables|GetGridSizeOverride_X)))
  (return (+ (/ (Tile|GetNumVisibleSmartCards ItemsToShow OfferVM GridSizeOverride IsFilteringActive) (select (> _gridsizeoverride_y 0) _gridsizeoverride_y _gridsize_y)) (Math|Integer|Clamp(Integer) (Math|Integer|%25%28Integer%29 (Tile|GetNumVisibleSmartCards ItemsToShow OfferVM GridSizeOverride IsFilteringActive) (select (> _gridsizeoverride_y 0) _gridsizeoverride_y _gridsize_y)) 0 1))))

(fn GetOfferDescriptionText
  (params (Description :type Text) (OfferType :type EFortItemShopOfferType%20Enum :default "LevelBundle"))
  (return (Utilities|Select :Index (Variables|GetOfferType) :LevelBundle (Variables|GetDescription) :BattlePass (Variables|GetDescription) :ProgressiveCosmetic (LocalVariables|GetProgressiveCosmeticShortDescriptionText) :Crew (LocalVariables|GetCrewShortDescriptionText) :Special (Variables|GetDescription) :Regular (Variables|GetDescription) :MainItemType (Variables|GetDescription) :MainItemTypeAndBonus (Variables|GetDescription) :Pack (Variables|GetDescription) :OutfitPack (Variables|GetDescription) :PackAndBonus (Variables|GetDescription) :Vehicle (Variables|GetDescription) :VehicleAndBonus (Variables|GetDescription) :Bundle (Variables|GetDescription) :LEGOKit (Variables|GetDescription))))

(fn GetProgressTextFromXpProgress
  (params (CurrentProgress :type Integer) (TotalProgress :type Integer))
  (return (Utilities|Text|FormatText "NSLOCTEXT(\"[9E291A03796F8DEF100F121C89307C5C]\", \"BA55F86E42D8CE1FC7D0E8B6BAEE3E99\", \"{earned}<text color=\\\"C8C8C8\\\"> / {required} XP</>\")" (Math|Integer|Min(Integer) CurrentProgress TotalProgress) (select (< TotalProgress 999) (Utilities|Text|ToText(Integer) TotalProgress) (Inventory|GetShorthandQuantity TotalProgress)))))

(fn GetRuleForBundleOffer
  (params (OfferViewModel :type Fort%20Item%20Shop%20Offer%20VM%20Object%20Reference))
  (bind _localtilesize (LocalVariables|GetLocalTileSize))
  (bind _returnvalue (Utilities|Select :Index _localtilesize :Size_1_x_1 "1_x_1" :Size_2_x_1 "2_x_1" :Size_3_x_1 "3_x_1" :Size_4_x_1 "4_x_1" :Size_5_x_1 "5_x_1" :Size_1_x_2 "1_x_2" :Size_2_x_2 "2_x_2" :Size_3_x_2 "3_x_2" :Size_4_x_2 "4_x_2" :Size_5_x_2 "5_x_2"))
  (bind _returnvalue_2 (Utilities|String|ToString(Name) _returnvalue))
  (bind _returnvalue_3 (Utilities|String|Append _returnvalue_2 "_JamTrack"))
  (bind _isjamtrackbundle (LocalVariables|GetIsJamTrackBundle))
  (bind _returnvalue_4 (select _isjamtrackbundle _returnvalue_3 _returnvalue_2))
  (LocalVariables|SetLocalTileSize (Variables|/Script/FortniteUI.FortItemShopOfferVM|GetTileSize OfferViewModel))
  ; check if jam track bundle
  (bind _array_element (Utilities|Array|ForEachLoopwithBreak (Variables|/Script/FortniteUI.FortItemShopOfferVM|GetGrantedItems (Variables|GetOfferViewModel)) :!label _label)
    (:LoopBody
      (bind _output_get (LocalVariables|SetIsJamTrackBundle (Utilities|Enum|Equal(Enum) (Class|FortItemFunctionLibrary|GetItemType (Variables|/Script/FortniteUI.FortItemVM|GetItemDefinition _array_element)) "SparksSong")))
      (if (not _output_get)
        (goto _label Break)))
    (:Completed
      _))
  (bind _returnvalue_5 (Utilities|GetDataTableRow "/FortUILibrary/Systems/ItemShopTile/Rulesets/DT_ItemShop_ImageRuleset_Bundle.DT_ItemShop_ImageRuleset_Bundle" (Utilities|String|StringToName _returnvalue_4))
    (:then
      (return _returnvalue_5 _returnvalue_4))
    (:RowNotFound
      (Development|PrintString 0 (Utilities|String|Append "No rule found for " _returnvalue_4) false))))

(fn GetRuleForItemType
  (params (ItemViewModel :type Fort%20Item%20VM%20Object%20Reference) (TileSize :type EItemShopTileSize%20Enum :default "Size_1_x_1") (ShopViewModel :type Fort%20Item%20Shop%20VM%20Object%20Reference))
  (bind _itemviewmodel (Variables|GetItemViewModel))
  (bind _returnvalue (Class|FortItemShopBlueprintLibrary|GetItemShopTileImageRuleset _itemviewmodel))
  (bind (_lefts _rights) (Utilities|String|Split (Utilities|GetDisplayName (Utilities|Select :Index _returnvalue :Default "/FortUILibrary/Systems/ItemShopTile/Rulesets/DT_ItemShop_ImageRuleset_Default.DT_ItemShop_ImageRuleset_Default" :ScaledStandard "/FortUILibrary/Systems/ItemShopTile/Rulesets/DT_ItemShop_ImageRuleset_ScaledStandard.DT_ItemShop_ImageRuleset_ScaledStandard" :ScaledHalfStep "/FortUILibrary/Systems/ItemShopTile/Rulesets/DT_ItemShop_ImageRuleset_ScaledHalfStep.DT_ItemShop_ImageRuleset_ScaledHalfStep" :Bundle "/FortUILibrary/Systems/ItemShopTile/Rulesets/DT_ItemShop_ImageRuleset_Bundle.DT_ItemShop_ImageRuleset_Bundle" :FullSquare "/FortUILibrary/Systems/ItemShopTile/Rulesets/DT_ItemShop_ImageRuleset_FullSquare.DT_ItemShop_ImageRuleset_FullSquare" :Basic "/FortUILibrary/Systems/ItemShopTile/Rulesets/DT_ItemShop_ImageRuleset_Basic.DT_ItemShop_ImageRuleset_Basic" :ScaledFull "/FortUILibrary/Systems/ItemShopTile/Rulesets/DT_ItemShop_ImageRuleset_ScaledFull.DT_ItemShop_ImageRuleset_ScaledFull" :FullTileSpan "/FortUILibrary/Systems/ItemShopTile/Rulesets/DT_ItemShop_ImageRuleset_FullTileSpan.DT_ItemShop_ImageRuleset_FullTileSpan" :Ruleset_Kicks "/FortUILibrary/Systems/ItemShopTile/Rulesets/DT_ItemShop_ImageRuleset_Kicks.DT_ItemShop_ImageRuleset_Kicks" :Ruleset_Sidekicks "/FortUILibrary/Systems/ItemShopTile/Rulesets/DT_ItemShop_ImageRuleset_Sidekicks.DT_ItemShop_ImageRuleset_Sidekicks")) "DT_ItemShop_ImageRuleset_"))
  (bind _returnvalue_2 (Utilities|Select :Index _returnvalue :Default "/FortUILibrary/Systems/ItemShopTile/Rulesets/DT_ItemShop_ImageRuleset_Default.DT_ItemShop_ImageRuleset_Default" :ScaledStandard "/FortUILibrary/Systems/ItemShopTile/Rulesets/DT_ItemShop_ImageRuleset_ScaledStandard.DT_ItemShop_ImageRuleset_ScaledStandard" :ScaledHalfStep "/FortUILibrary/Systems/ItemShopTile/Rulesets/DT_ItemShop_ImageRuleset_ScaledHalfStep.DT_ItemShop_ImageRuleset_ScaledHalfStep" :Bundle "/FortUILibrary/Systems/ItemShopTile/Rulesets/DT_ItemShop_ImageRuleset_Bundle.DT_ItemShop_ImageRuleset_Bundle" :FullSquare "/FortUILibrary/Systems/ItemShopTile/Rulesets/DT_ItemShop_ImageRuleset_FullSquare.DT_ItemShop_ImageRuleset_FullSquare" :Basic "/FortUILibrary/Systems/ItemShopTile/Rulesets/DT_ItemShop_ImageRuleset_Basic.DT_ItemShop_ImageRuleset_Basic" :ScaledFull "/FortUILibrary/Systems/ItemShopTile/Rulesets/DT_ItemShop_ImageRuleset_ScaledFull.DT_ItemShop_ImageRuleset_ScaledFull" :FullTileSpan "/FortUILibrary/Systems/ItemShopTile/Rulesets/DT_ItemShop_ImageRuleset_FullTileSpan.DT_ItemShop_ImageRuleset_FullTileSpan" :Ruleset_Kicks "/FortUILibrary/Systems/ItemShopTile/Rulesets/DT_ItemShop_ImageRuleset_Kicks.DT_ItemShop_ImageRuleset_Kicks" :Ruleset_Sidekicks "/FortUILibrary/Systems/ItemShopTile/Rulesets/DT_ItemShop_ImageRuleset_Sidekicks.DT_ItemShop_ImageRuleset_Sidekicks"))
  (bind _rulesetvalues (Tile|GetRuleforRuleset _returnvalue_2 (Variables|GetTileSize)))
  (return _rulesetvalues _rights))

(fn GetRuleForRuleset
  (params (DataTable :type Data%20Table%20Object%20Reference) (TileSize :type EItemShopTileSize%20Enum :default "Size_1_x_1"))
  (bind _returnvalue (Utilities|GetDataTableRow DataTable (Utilities|Select :Index TileSize :Size_1_x_1 "1_x_1" :Size_2_x_1 "2_x_1" :Size_3_x_1 "3_x_1" :Size_4_x_1 "4_x_1" :Size_5_x_1 "5_x_1" :Size_1_x_2 "1_x_2" :Size_2_x_2 "2_x_2" :Size_3_x_2 "3_x_2" :Size_4_x_2 "4_x_2" :Size_5_x_2 "5_x_2"))
    (:then
      _)
    (:RowNotFound
      (Development|PrintString 0 "Rule value not found for tile size")))
  (return _returnvalue))

(fn GetRuleForRulesetWithSmartCards
  (params (OfferViewModel :type Fort%20Item%20Shop%20Offer%20VM%20Object%20Reference) (Ruleset :type Data%20Table%20Object%20Reference) (ItemsToShow :type Array%20of%20Fort%20Item%20VM%20Object%20References) (IsFilteringActive :type Boolean))
  (bind _offerviewmodel (Variables|GetOfferViewModel))
  (bind _returnvalue (Tile|GetNumVisibleSmartCardColumns :ItemsToShow ItemsToShow :OfferVM _offerviewmodel :IsFilteringActive IsFilteringActive))
  (bind _returnvalue_2 (Utilities|String|ToString(Integer) _returnvalue))
  (bind _localtilesize (LocalVariables|GetLocalTileSize))
  (bind _returnvalue_3 (Utilities|Select :Index _localtilesize :Size_1_x_1 "1_x_1" :Size_2_x_1 "2_x_1" :Size_3_x_1 "3_x_1" :Size_4_x_1 "4_x_1" :Size_5_x_1 "5_x_1" :Size_1_x_2 "1_x_2" :Size_2_x_2 "2_x_2" :Size_3_x_2 "3_x_2" :Size_4_x_2 "4_x_2" :Size_5_x_2 "5_x_2"))
  (bind _returnvalue_4 (Utilities|String|ToString(Name) _returnvalue_3))
  (bind _returnvalue_5 (Utilities|String|Append _returnvalue_4 "_" _returnvalue_2 "Col"))
  (LocalVariables|SetLocalTileSize (Variables|/Script/FortniteUI.FortItemShopOfferVM|GetTileSize OfferViewModel))
  ;[0 0 0] first, we try and build a string for smart cards
  (bind _returnvalue_6 (Utilities|GetDataTableRow (Variables|GetRuleset) (Utilities|String|StringToName _returnvalue_5))
    (:then
      (return _returnvalue_6 _returnvalue_5))
    (:RowNotFound
      ;[0 0 0] if that doesn't work, we look for the normal tile size name
      (bind _returnvalue_7 (Utilities|GetDataTableRow (Variables|GetRuleset) _returnvalue_3)
        (:then
          (return _returnvalue_7 _returnvalue_4))
        (:RowNotFound
          (Development|PrintString 0 (Utilities|String|Append "No rule found for " _returnvalue_5) false))))))

(fn GetSectionFromLayoutStruct
  (params (InLayoutStruct :type Fort%20Item%20Shop%20Layout%20Structure) (InItemShopVM :type Fort%20Item%20Shop%20VM%20Object%20Reference))
  (bind _inlayoutstruct (Variables|GetInLayoutStruct))
  (bind (_sectionid _subsectionid) (Utilities|Struct|BreakFortItemShopLayout _inlayoutstruct))
  (bind _returnvalue (Class|FortItemShopVM|GetSection InItemShopVM _sectionid true))
  (bind _output_get (LocalVariables|SetFoundSection _returnvalue))
  (Utilities|IsValid _output_get
    (:Is%20Valid
      (bind _output_get_2 (LocalVariables|SetFoundSubSection (Class|FortItemShopVM|GetSubsection (Variables|GetInItemShopVM) _subsectionid _returnvalue)))
      (Utilities|IsValid _output_get_2
        (:Is%20Valid
          (return (LocalVariables|GetFoundSubSection)))
        (:Is%20Not%20Valid
          (return (LocalVariables|GetFoundSection)))))
    (:Is%20Not%20Valid
      (return 0))))

(fn GetTitleColorsFromStringMetadata
  (params (StringMetadata :type Map%20of%20Names%20to%20Strings))
  (bind (_value _returnvalue) (Utilities|Map|Find StringMetadata "titleColorB"))
  (bind (_value_2 _returnvalue_2) (Utilities|Map|Find StringMetadata "titleColorA"))
  (return (Class|UIKitFunctionLibrary|HextoColor _value_2 (Variables|Get__WorldContext)) (Class|UIKitFunctionLibrary|HextoColor _value (Variables|Get__WorldContext)) (and _returnvalue_2 _returnvalue)))

(fn IsFrameBreakEnabled
  (params (ShouldFrameBreak :type Boolean) (IsFilterActive :type Boolean))
  (return (Math|Boolean|ANDBoolean (Class|FortItemShopBlueprintLibrary|IsFrameBreakAllowed) (not (Variables|GetIsFilterActive)) (or (GlobalConfigurationData|GetConfigDataBoolwithDefault "shop_frame_break_all") (Variables|GetShouldFrameBreak)))))

(fn IsItemPreviewEnabledForOfferGroup
  (params (OfferGroupVM :type Fort%20Item%20Shop%20Offer%20Group%20VM%20Object%20Reference))
  (Utilities|IsValid OfferGroupVM
    (:Is%20Valid
      (return (Utilities|Select :Index (Class|FortItemShopOfferGroupVM|GetDisplayType OfferGroupVM) :TileGrid true :ExpandableList true :Billboard false :ScrollableList false)))
    (:Is%20Not%20Valid
      (return false))))

(fn IsItemPreviewEnabledForSelectedOfferGroup
  (params (SelectionVM :type Fort%20Item%20Shop%20Selection%20VM%20Object%20Reference))
  (Utilities|IsValid SelectionVM
    (:Is%20Valid
      (return (Utility|IsItemPreviewEnabledforOfferGroup (Variables|/Script/FortniteUI.FortItemShopSelectionVM|GetSelectedOfferGroup SelectionVM) (Variables|Get__WorldContext))))
    (:Is%20Not%20Valid
      (return false))))

(fn SetPurchaseButtonStyle
  (params (Purchase%20Button :type Fort%20CTAButton%20Object%20Reference) (GetVbucks%20Button :type Fort%20CTAButton%20Object%20Reference))
  (bind _array (Utilities|Array|MakeArray))
  (bind _getvbucks_button (Variables|GetGetVbucks%20Button))
  (bind _stylevarianttag (LocalVariables|GetStyleVariantTag))
  (bind _returnvalue (Development|GetConsoleVariableBoolValue "Store.EnablePDPPurchaseButtonVariants"))
  (if _returnvalue
    (switch string (GlobalConfigurationData|GetConfigDataStringwithDefault "pdp_purchase_button_variants" "base yellow color")
      (:Default
        _)
      (:base%20yellow%20color
        _)
      (:var1%20-%20orange
        (LocalVariables|SetStyleVariantTag (Utilities|Array|MakeArray "(TagName=\"StyleSheet.Style.Purchase.A\")"))
        (StyleSheets|UpdateWidgetStyleTags (Variables|GetPurchase%20Button) _stylevarianttag _array :!label _label)
        (Utilities|IsValid _getvbucks_button
          (:Is%20Valid
            (StyleSheets|UpdateWidgetStyleTags _getvbucks_button _stylevarianttag _array))))
      (:var2%20-%20blue
        (LocalVariables|SetStyleVariantTag (Utilities|Array|MakeArray "(TagName=\"StyleSheet.Style.Purchase.B\")"))
        (goto _label))
      (:var3%20-%20white
        (LocalVariables|SetStyleVariantTag (Utilities|Array|MakeArray "(TagName=\"StyleSheet.Style.Purchase.C\")"))
        (goto _label))
      (:var4%20-%20green
        (LocalVariables|SetStyleVariantTag (Utilities|Array|MakeArray "(TagName=\"StyleSheet.Style.Purchase.D\")"))
        (goto _label)))
    (return)))

(fn SetTileBackgroundColors
  (params (MID :type Material%20Instance%20Dynamic%20Object%20Reference) (FortItemShopOfferMetaDisplayData :type Fort%20Item%20Shop%20Offer%20Meta%20Display%20Data%20Structure%20%28by%20ref%29))
  (bind (_color1 _color2 _color3) (Utilities|Struct|BreakFortItemShopOfferMetaDisplayData FortItemShopOfferMetaDisplayData))
  (Rendering|Material|SetVectorParameterValue MID "Background Color 1" _color1 :!class /Script/Engine.MaterialInstanceDynamic)
  (Rendering|Material|SetVectorParameterValue MID "Background Color 2" _color2 :!class /Script/Engine.MaterialInstanceDynamic)
  (Rendering|Material|SetVectorParameterValue MID "Background Color 3" _color3 :!class /Script/Engine.MaterialInstanceDynamic)
  (Rendering|Material|SetScalarParameterValue MID "bUse3Colors" (Math|Float|SelectFloat 0.0 1.0 (== _color3 "(R=0.000000,G=0.000000,B=0.000000,A=0.000000)")) :!class /Script/Engine.MaterialInstanceDynamic))

(fn ShouldShowCpcDialog
  (params (Context :type Widget%20Object%20Reference) (OfferVM :type Fort%20Item%20Shop%20Offer%20VM%20Object%20Reference))
  (bind _offervm (Variables|GetOfferVM)
    (:then
      (bind _returnvalue (Class|FortItemShopBlueprintLibrary|ShouldShowRightOfWithdrawalModal (Variables|GetContext)))
      (return (and _returnvalue (Utilities|Enum|Equal(Enum) (BaseOfferVM|GetCurrencyType _offervm) "Mtx"))))
    (:else
      (return false))))

(fn ShouldUseVehicleRuleset
  (params (OfferVM :type Fort%20Item%20Shop%20Offer%20VM%20Object%20Reference))
  (bind _loopitem (LocalVariables|GetLoopItem))
  (bind _returnvalue (Class|FortItemVM|GetItemTemplateId _loopitem))
  (bind _returnvalue_2 (Utilities|String|Contains _returnvalue "VehicleCosmetics_Body"))
  (bind _array_element (Utilities|Array|ForEachLoopwithBreak (Variables|/Script/FortniteUI.FortItemShopOfferVM|GetDisplayedPrimaryGrantedItems OfferVM) :!label _label)
    (:LoopBody
      (LocalVariables|SetLoopItem _array_element)
      (if (not (LocalVariables|GetContainsBody))
        (bind _output_get (LocalVariables|SetContainsBody _returnvalue_2))
        (if _output_get
          _
          (else
            _)))
      (bind _output_get_2 (LocalVariables|SetContainsAllowedTypesOnly (Math|Boolean|ORBoolean _returnvalue_2 (Utilities|String|Contains _returnvalue "VehicleCosmetics_Skin") (Utilities|String|Contains _returnvalue "VehicleCosmetics_Booster") (Utilities|String|Contains _returnvalue "VehicleCosmetics_Wheel") (Utilities|String|Contains _returnvalue "VehicleCosmetics_DriftTrail"))))
      (if (not _output_get_2)
        (goto _label Break)))
    (:Completed
      (return (and (LocalVariables|GetContainsBody) (LocalVariables|GetContainsAllowedTypesOnly))))))
