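; Blueprint interface for the item details portion of the Item Shop product detail page (PDP) in the multi-product frontend UI.
; It declares paired inject and clear functions for four slots: ALC item list, custom widget, ISBs and purchase block.
; Each inject function receives a widget object reference and each clear function takes no parameters, so callers can supply content without depending on the implementing widget's class.
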(fn ClearALCListInjection)

(fn ClearCustomInjection)

(fn ClearISBInjection)

(fn ClearPurchaseBlockInjection)

(fn InjectALCList
  (params (ItemListALC :type Widget%20Object%20Reference)))

(fn InjectCustomWidget
  (params (CustomWidget :type Widget%20Object%20Reference)))

(fn InjectISBs
  (params (ISBs :type Widget%20Object%20Reference)))

(fn InjectPurchaseBlock
  (params (PurchaseBlock :type Widget%20Object%20Reference)))
