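; Item Shop multi-product preview world that shows background and floor behind previewed items, with animated transitions between them.
; Current/Next dynamic materials take a loaded texture or material; timelines drive shared parameter-collection values to ease out, then reveal Next, which becomes Current.
; Quick scrolling caches the loaded resource and defers ease-in until scrolling stops.
; Item type and on-pawn state select presets for Next's background parameters and floor lighting channels, with platform overrides for lighting.
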(event EventOnNewSceneBackgroundChildActor
  (params (NewActor :type Fort%20Static%20Mesh%20Actor%20Object%20Reference))
  ;[0 0 0] Initialization
  (CallFunction|SetBackgroundFromChildActor)
  (CallFunction|AssignBackgroundMaterial (Variables|Default|GetDefaultBackgroundMaterial) true)
  (CallFunction|AssignBackgroundMaterial (Variables|Default|GetDefaultBackgroundMaterial))
  (Variables|Default|SetIsPendingLoadedResource false)
  (Variables|Default|SetIsQuickScrolling false)
  (Variables|Default|SetPendingLoadedResource 0)
  (Variables|Default|SetAllowEaseOutTransition true))

(event EventBeginPlay
  (bind _defaultfloormaterial (Variables|Default|GetDefaultFloorMaterial))
  ;[0 0 0] Initialization
  (CallFunction|AssignFloorMaterial _defaultfloormaterial true)
  (CallFunction|AssignFloorMaterial _defaultfloormaterial)
  (CallFunction|SetBackgroundFromChildActor)
  (CallFunction|AssignBackgroundMaterial (Variables|Default|GetDefaultBackgroundMaterial) true)
  (CallFunction|AssignBackgroundMaterial (Variables|Default|GetDefaultBackgroundMaterial))
  (Variables|Default|SetIsPendingLoadedResource false)
  (Variables|Default|SetIsQuickScrolling false)
  (Variables|Default|SetPendingLoadedResource 0)
  (Variables|Default|SetAllowEaseOutTransition true))

(event EventOnUpdateDisplay
  (params (bShowFloor :type Boolean) (bShowEffects :type Boolean) (bIsOnPawn :type Boolean) (ItemType :type EFortItemType%20Enum :default "WorldItem"))
  ;[0 0 0] Update floor visibility and lighting setup (whenever a new Item is being previewed)
  (Variables|Default|SetIsFloorVisible bShowFloor)
  (Variables|Default|SetIs%20on%20Pawn bIsOnPawn)
  (Variables|Default|SetPreviewedItemType ItemType)
  (Utilities|FlowControl|Sequence
    (:then_0
      ;;[0.3958 0.3958 0.3958 0.922] Floor visiblity
      (Rendering|SetVisibility (Variables|Default|GetFloor_Current) (Variables|Default|GetIsFloorVisible) true)
      (if (Variables|Default|GetIsFloorVisible)
        (bind (_direction _floormask) (Timeline|FloorVisibility PlayFromStart)
          (:Update
            (Rendering|Material|SetScalarParameterValue (Variables|Default|GetFloorMID_Next) "Floor Reveal Transition" _floormask :!class /Script/Engine.MaterialInstanceDynamic))) ; animate floor mask so that we hide the popping of the floor
      ))
    (:then_1
      ;;[0.3958 0.3958 0.3958 0.922] Lighting and material setup depending on previewed cosmetic type
      (CallFunction|DefinePresettoUse)
      (CallFunction|ApplyLightingChannels)
      (CallFunction|ApplyMaterialInstanceParameters))))

(event EventOnUpdateBackgroundRequested
  (params (TransitionDirection :type EVaultWorldTransitionDirection%20Enum :default "Forward"))
  (bind _transitiondirection (Variables|Default|GetTransitionDirection))
  ;[0 0 0] We received an Update Background request: ease out the current world
  (Variables|Default|SetTransitionDirection TransitionDirection)
  (if (Components|Timeline|IsPlaying (Variables|MPItemShop_VaultWorld|GetEaseInNext))
    ;;[0 0 0 1 14] if an Ease In transition is playing, we stop it and reset to final state
    (Components|Timeline|Stop (Variables|MPItemShop_VaultWorld|GetEaseInNext))
    (CallFunction|OnEaseinTransitionFinished))
  (CallFunction|SetupTransitionAngle _transitiondirection)
  (if (Utilities|Enum|Equal(Enum) _transitiondirection "None")
    ;;[0 0 0] Ease Out Current
    (Rendering|Material|SetScalarParameterValue "/Game/Environments/FrontEnd/Blueprints/ItemPreview/VaultWorldTransitions/MaterialCollectionParameters/MPC_UI_Master_VaultWorld_Settings.MPC_UI_Master_VaultWorld_Settings" "Transition Loading" 1.0 :!class /Script/Engine.KismetMaterialLibrary)
    (elif (Variables|Default|GetAllowEaseOutTransition)
      (Variables|Default|SetAllowEaseOutTransition false)
      (bind (_direction _transitionreveal) (Timeline|EaseOutCurrent PlayFromStart)
        (:Update
          (Rendering|Material|SetScalarParameterValue "/Game/Environments/FrontEnd/Blueprints/ItemPreview/VaultWorldTransitions/MaterialCollectionParameters/MPC_UI_Master_VaultWorld_Settings.MPC_UI_Master_VaultWorld_Settings" "Transition Loading" _transitionreveal :!class /Script/Engine.KismetMaterialLibrary))))))

(event EventOnTransitionBackground
  (params (TransitionDirection :type EVaultWorldTransitionDirection%20Enum :default "Forward") (IntensityTransition :type EBackgroundIntensityLevel%20Enum))
  ;[0 0 0] Next background is loaded: if we're not quick scrolling, ease in Next
  (if (and (not (Variables|Default|GetIsPendingLoadedResource)) (not (Variables|Default|GetIsQuickScrolling)))
    (CallFunction|OnEaseIn)))

(event EventOnSetupTextureBackground
  (params (LoadedBackgroundTexture :type Texture%202D%20Object%20Reference) (BackgroundInfo :type Vault%20World%20Background%20Data%20Structure%20%28by%20ref%29))
  ;[0 0 0] Next background (texture or material) is loaded: if not quick scrolling, assign it to Next. If quick scrolling, cache it for later.
  ;;[0.3958 0.3958 0.3958 0.922] Texture
  (if (Variables|Default|GetIsQuickScrolling)
    (Variables|Default|SetPendingLoadedResource LoadedBackgroundTexture)
    (Variables|Default|SetIsPendingLoadedResource true)
    (else
      (CallFunction|HandleTextureBackground LoadedBackgroundTexture))))

(event EventOnSetupMaterialBackground
  (params (LoadedBackgroundMaterial :type Material%20Interface%20Object%20Reference) (BackgroundInfo :type Vault%20World%20Background%20Data%20Structure%20%28by%20ref%29))
  ;[0 0 0] Next background (texture or material) is loaded: if not quick scrolling, assign it to Next. If quick scrolling, cache it for later.
  ;;[0.3958 0.3958 0.3958 0.922] Material
  (if (Variables|Default|GetIsQuickScrolling)
    (Variables|Default|SetPendingLoadedResource LoadedBackgroundMaterial)
    (Variables|Default|SetIsPendingLoadedResource true)
    (else
      (CallFunction|HandleMaterialBackground LoadedBackgroundMaterial))))

(event Custom|On%20Ease%20In
  ;[0 0 0] Ease In Next
  (Components|Timeline|Stop (Variables|MPItemShop_VaultWorld|GetEaseOutCurrent))
  (switch Utilities|FlowControl|Switch|SwitchonEVaultWorldTransitionDirection (Variables|Default|GetTransitionDirection)
    (:Forward
      _)
    (:Backward
      _)
    (:None
      (CallFunction|OnEaseinTransitionFinished)))
  (bind (_direction _transitionreveal) (Timeline|EaseInNext PlayFromStart)
    (:Update
      (Rendering|Material|SetScalarParameterValue "/Game/Environments/FrontEnd/Blueprints/ItemPreview/VaultWorldTransitions/MaterialCollectionParameters/MPC_UI_Master_VaultWorld_Settings.MPC_UI_Master_VaultWorld_Settings" "Transition Reveal" _transitionreveal :!class /Script/Engine.KismetMaterialLibrary))
    (:Finished)))

(event Custom|On%20Ease%20In%20Transition%20Finished
  ;[0 0 0] Ease in is finished: Next background becomes current and we're back to the initial state
  ;;[0 0 0] Next becomes Current
  (CallFunction|AssignFloorMaterial (Variables|Default|GetFloorMID_Next) true)
  (CallFunction|AssignBackgroundMaterial (Variables|Default|GetBackgroundMID_Next) true)
  ;;[0 0 0] reset transition params
  (Rendering|Material|SetScalarParameterValue "/Game/Environments/FrontEnd/Blueprints/ItemPreview/VaultWorldTransitions/MaterialCollectionParameters/MPC_UI_Master_VaultWorld_Settings.MPC_UI_Master_VaultWorld_Settings" "Transition Reveal" :!class /Script/Engine.KismetMaterialLibrary)
  (Rendering|Material|SetScalarParameterValue "/Game/Environments/FrontEnd/Blueprints/ItemPreview/VaultWorldTransitions/MaterialCollectionParameters/MPC_UI_Master_VaultWorld_Settings.MPC_UI_Master_VaultWorld_Settings" "Transition Loading" :!class /Script/Engine.KismetMaterialLibrary)
  (Variables|Default|SetAllowEaseOutTransition true) ; now that we finished the Ease In, we can allow\nthe Ease Out transition to play next time
)

(event EventOnQuickScrollStateChanged
  (params (bInIsQuickScrolling :type Boolean))
  ;[0 0 0] Quick Scroll State: if finished, check if we have pending resources to assign and ease in
  (Utilities|FlowControl|DeltaGateWildcard (Variables|Default|GetIsQuickScrolling) bInIsQuickScrolling
    (:Success
      (if (and (not (Variables|Default|GetIsQuickScrolling)) (Variables|Default|GetIsPendingLoadedResource))
        (bind _astexture_2d (Utilities|Casting|CastToTexture2D (Variables|Default|GetPendingLoadedResource))
          (:then
            (CallFunction|HandleTextureBackground _astexture_2d))
          (:CastFailed
            (bind _asmaterial_interface (Utilities|Casting|CastToMaterialInterface (Variables|Default|GetPendingLoadedResource))
              (:then
                (CallFunction|HandleMaterialBackground _asmaterial_interface))
              (:CastFailed
                (CallFunction|HandleTextureBackground "") ; if we have a pending request but nothing is valid,\nmeans we need to set the default FN background
              ))))
        (Variables|Default|SetPendingLoadedResource 0)
        (Variables|Default|SetIsPendingLoadedResource false)
        (CallFunction|ApplyMaterialInstanceParameters)
        (CallFunction|OnEaseIn)))) ; execute only if state has changed
)

(fn ApplyLightingChannels
  (bind (_value _returnvalue) (Utilities|Map|Find (Variables|Default|GetFloorPresetMap) (Variables|Default|GetPreviewedItemType)))
  (bind (_value_floorpresetmi_2_090a6c2b4291cb61ed689c8b4d3e6a5f _value_uselightchannel0_7_e644d8e14b86f750b82df782421c31ba _value_uselightchannel1_9_ad76819c4be803269a4abe9ab657be4e) (Utilities|Struct|BreakSVaultWorldFloorSettings _value))
  (bind _floor_next (Variables|Default|GetFloor_Next))
  (bind (_fallbackfloorpreset_floorpresetmi_2_090a6c2b4291cb61ed689c8b4d3e6a5f _fallbackfloorpreset_uselightchannel0_7_e644d8e14b86f750b82df782421c31ba _fallbackfloorpreset_uselightchannel1_9_ad76819c4be803269a4abe9ab657be4e) (Utilities|Struct|BreakSVaultWorldFloorSettings (Variables|GetFallbackFloorPreset_FloorPresetMI_2_090A6C2B4291CB61ED689C8B4D3E6A5F)))
  (if (Math|Boolean|ORBoolean (TouchPlatform|IsNintendoSwitchorNintendoSwitchPreview) (Fortnite|IsMobilePlatform) (TouchPlatform|IsMobileGame) (Class|PlatformHelpersLibrary|IsPcPerfMode))
    (Rendering|Components|SetLightingChannels _floor_next true true false)
    (else
      (Rendering|Components|SetLightingChannels _floor_next (select (Variables|Default|GetIs%20on%20Pawn) _fallbackfloorpreset_uselightchannel0_7_e644d8e14b86f750b82df782421c31ba (select _returnvalue _value_uselightchannel0_7_e644d8e14b86f750b82df782421c31ba true)) (select (Variables|Default|GetIs%20on%20Pawn) _fallbackfloorpreset_uselightchannel1_9_ad76819c4be803269a4abe9ab657be4e (select _returnvalue _value_uselightchannel1_9_ad76819c4be803269a4abe9ab657be4e false)) false))))

(fn ApplyMaterialInstanceParameters
  (bind _backgroundmid_next (Variables|Default|GetBackgroundMID_Next))
  (bind _presetmi (Variables|Default|GetPresetMI)
    (:then
      (Utilities|FlowControl|Sequence
        (:then_0
          (for _array_element (Variables|/Script/Engine.MaterialInstance|GetScalarParameterValues _presetmi)
            (bind (_array_element_parameterinfo _array_element_parametervalue) (Utilities|Struct|BreakScalarParameterValue _array_element))
            (bind _output_get (LocalVariables|SetLocalParamInfo _array_element_parameterinfo))
            (Rendering|Material|SetScalarParameterValuebyInfo _backgroundmid_next _output_get _array_element_parametervalue)))
        (:then_1
          (for _array_element_2 (Variables|/Script/Engine.MaterialInstance|GetVectorParameterValues _presetmi)
            (bind (_array_element_parameterinfo_2 _array_element_parametervalue_2) (Utilities|Struct|BreakVectorParameterValue _array_element_2))
            (bind _output_get_2 (LocalVariables|SetLocalParamInfo _array_element_parameterinfo_2))
            (Rendering|Material|SetVectorParameterValuebyInfo _backgroundmid_next _output_get_2 _array_element_parametervalue_2)))))
    (:else)))

(fn AssignBackgroundMaterial
  (params (SourceMaterial :type Material%20Interface%20Object%20Reference) (IsCurrent :type Boolean))
  (bind _iscurrent (Variables|GetIsCurrent))
  (bind _returnvalue (Math|Float|SelectFloat 0.0 1.0 _iscurrent))
  (if IsCurrent
    (bind _returnvalue_2 (Rendering|Material|CreateDynamicMaterialInstance (Variables|Default|GetBackground_Current) 0 SourceMaterial "None" :!class /Script/Engine.PrimitiveComponent))
    (bind _output_get (Variables|Default|SetBackgroundMID_Current _returnvalue_2))
    (Rendering|Material|SetScalarParameterValue _output_get "Invert" _returnvalue :!class /Script/Engine.MaterialInstanceDynamic)
    (else
      (bind _returnvalue_3 (Rendering|Material|CreateDynamicMaterialInstance (Variables|Default|GetBackground_Next) 0 SourceMaterial "None" :!class /Script/Engine.PrimitiveComponent))
      (bind _output_get_2 (Variables|Default|SetBackgroundMID_Next _returnvalue_3))
      (Rendering|Material|SetScalarParameterValue _output_get_2 "Invert" _returnvalue :!class /Script/Engine.MaterialInstanceDynamic))))

(fn AssignFloorMaterial
  (params (SourceMaterial :type Material%20Interface%20Object%20Reference) (IsCurrent :type Boolean))
  (bind _iscurrent (Variables|GetIsCurrent))
  (bind _returnvalue (Math|Float|SelectFloat 0.0 1.0 _iscurrent))
  (if IsCurrent
    (bind _returnvalue_2 (Rendering|Material|CreateDynamicMaterialInstance (Variables|Default|GetFloor_Current) 0 SourceMaterial "None" :!class /Script/Engine.PrimitiveComponent))
    (bind _output_get (Variables|Default|SetFloorMID_Current _returnvalue_2))
    (Rendering|Material|SetScalarParameterValue _output_get "Invert" _returnvalue :!class /Script/Engine.MaterialInstanceDynamic)
    (CallFunction|SetFloorMaterialDynamic _output_get)
    (else
      (bind _returnvalue_3 (Rendering|Material|CreateDynamicMaterialInstance (Variables|Default|GetFloor_Next) 0 SourceMaterial "None" :!class /Script/Engine.PrimitiveComponent))
      (bind _output_get_2 (Variables|Default|SetFloorMID_Next _returnvalue_3))
      (Rendering|Material|SetScalarParameterValue _output_get_2 "Invert" _returnvalue :!class /Script/Engine.MaterialInstanceDynamic)
      (CallFunction|SetFloorMaterialDynamic _output_get_2))))

(fn DefinePresetToUse
  (bind (_value _returnvalue) (Utilities|Map|Find (Variables|Default|GetFloorPresetMap) (Variables|Default|GetPreviewedItemType)))
  (bind _value_floorpresetmi_2_090a6c2b4291cb61ed689c8b4d3e6a5f (Utilities|Struct|BreakSVaultWorldFloorSettings _value))
  (bind _fallbackfloorpreset_floorpresetmi_2_090a6c2b4291cb61ed689c8b4d3e6a5f (Utilities|Struct|BreakSVaultWorldFloorSettings (Variables|GetFallbackFloorPreset_FloorPresetMI_2_090A6C2B4291CB61ED689C8B4D3E6A5F)))
  (bind _output_get (Variables|Default|SetPresetMI (select (Variables|Default|GetIs%20on%20Pawn) _fallbackfloorpreset_floorpresetmi_2_090a6c2b4291cb61ed689c8b4d3e6a5f (select _returnvalue _value_floorpresetmi_2_090a6c2b4291cb61ed689c8b4d3e6a5f _fallbackfloorpreset_floorpresetmi_2_090a6c2b4291cb61ed689c8b4d3e6a5f))))
  (return _output_get))

(fn HandleMaterialBackground
  (params (MaterialBackground :type Material%20Interface%20Object%20Reference) (IsCurrent :type Boolean))
  (bind _iscurrent (Variables|GetIsCurrent))
  (bind _iscurrent_2 (Variables|GetIsCurrent))
  (Utilities|FlowControl|Sequence
    (:then_0
      ;[0 0 0] setup background
      (CallFunction|AssignBackgroundMaterial (Variables|GetMaterialBackground) _iscurrent)
      (CallFunction|SetupBackgroundMaterial "" (select _iscurrent (Variables|Default|GetBackgroundMID_Current) (Variables|Default|GetBackgroundMID_Next))))
    (:then_1
      ;[0 0 0] setup floor
      (CallFunction|AssignFloorMaterial (Variables|GetMaterialBackground) _iscurrent_2)
      (CallFunction|SetupBackgroundMaterial "" (select _iscurrent_2 (Variables|Default|GetFloorMID_Current) (Variables|Default|GetFloorMID_Next)) true))
    (:then_2
      (CallFunction|ApplyMaterialInstanceParameters)
      (Rendering|Material|SetScalarParameterValue (Variables|Default|GetBackgroundMID_Next) "IsUsingShopTexture" 0.0 :!class /Script/Engine.MaterialInstanceDynamic))))

(fn HandleTextureBackground
  (params (TextureBackground :type Texture%202D%20Object%20Reference%20%28by%20ref%29) (IsCurrent :type Boolean))
  (bind _iscurrent (Variables|GetIsCurrent))
  (Utilities|FlowControl|Sequence
    (:then_0
      ;[0 0 0] Assign default material
      (CallFunction|AssignBackgroundMaterial (Variables|Default|GetDefaultBackgroundMaterial) _iscurrent)
      ;[0 0 0] Setup floor and background materials
      (CallFunction|SetupBackgroundMaterial (Variables|GetTextureBackground) (select (Variables|GetIsCurrent) (Variables|Default|GetBackgroundMID_Current) (Variables|Default|GetBackgroundMID_Next)) false true))
    (:then_1
      (CallFunction|AssignFloorMaterial (Variables|Default|GetDefaultFloorMaterial) _iscurrent)
      (CallFunction|SetupBackgroundMaterial (Variables|GetTextureBackground) (select (Variables|GetIsCurrent) (Variables|Default|GetFloorMID_Current) (Variables|Default|GetFloorMID_Next)) true true))
    (:then_2
      (CallFunction|ApplyMaterialInstanceParameters)
      (Rendering|Material|SetScalarParameterValue (Variables|Default|GetBackgroundMID_Next) "IsUsingShopTexture" 1.0 :!class /Script/Engine.MaterialInstanceDynamic))))

(fn SetBackgroundfromChildActor
  (Utilities|FlowControl|Sequence
    (:then_0
      (bind _asfort_static_mesh_actor (Utilities|Casting|CastToFortStaticMeshActor (Variables|/Script/Engine.ChildActorComponent|GetChildActor (Variables|FortItemPreviewWorld|GetSceneBackgroundComponent))))
      (Variables|Default|SetBackground_Current (Variables|/Script/Engine.StaticMeshActor|GetStaticMeshComponent _asfort_static_mesh_actor)))
    (:then_1
      (bind _asfort_static_mesh_actor_2 (Utilities|Casting|CastToFortStaticMeshActor (Variables|/Script/Engine.ChildActorComponent|GetChildActor (Variables|Default|GetSceneBackgroundComponent_Next))))
      (Variables|Default|SetBackground_Next (Variables|/Script/Engine.StaticMeshActor|GetStaticMeshComponent _asfort_static_mesh_actor_2)))))

(fn SetupBackgroundMaterial
  (params (TextureBackground :type Texture%202D%20Object%20Reference%20%28by%20ref%29) (MID :type Material%20Instance%20Dynamic%20Object%20Reference) (IsFloor :type Boolean) (FirstTimeSetup :type Boolean))
  (Rendering|Material|SetScalarParameterValue (Variables|GetMID) "IsFloor" (select IsFloor 1.0 0.0) :!class /Script/Engine.MaterialInstanceDynamic)
  (Utilities|IsValid (Variables|GetTextureBackground)
    (:Is%20Valid
      ;[0 0 0] texture valid: show texture
      (Rendering|Material|SetTextureParameterValue (Variables|GetMID) "BackgroundTexture" (Variables|GetTextureBackground))
      (Rendering|Material|SetScalarParameterValue (Variables|GetMID) "UseFallbackGradient" 0.0 :!class /Script/Engine.MaterialInstanceDynamic))
    (:Is%20Not%20Valid
      ;[0 0 0] texture not valid: show fallback gradient
      (Rendering|Material|SetTextureParameterValue (Variables|GetMID) "BackgroundTexture" "/Game/Athena/UI/Frontend/Spatial/Art/T_Ui_Empty.T_Ui_Empty")
      (Rendering|Material|SetScalarParameterValue (Variables|GetMID) "UseFallbackGradient" 1.0 :!class /Script/Engine.MaterialInstanceDynamic))))

(fn SetupTransitionAngle
  (params (TransitionDirection :type EVaultWorldTransitionDirection%20Enum :default "Forward"))
  (bind _returnvalue (Rendering|Material|GetScalarParameterValue "/Game/Environments/FrontEnd/Blueprints/ItemPreview/VaultWorldTransitions/MaterialCollectionParameters/MPC_UI_Master_VaultWorld_Settings.MPC_UI_Master_VaultWorld_Settings" "Is Horizontal Orientation" :!class /Script/Engine.KismetMaterialLibrary))
  (Rendering|Material|SetScalarParameterValue "/Game/Environments/FrontEnd/Blueprints/ItemPreview/VaultWorldTransitions/MaterialCollectionParameters/MPC_UI_Master_VaultWorld_Settings.MPC_UI_Master_VaultWorld_Settings" "Transition Direction" (Math|Float|Lerp (Utilities|Select :Index TransitionDirection :Forward 0.029 :Backward 0.509 :None 0.0) (Utilities|Select :Index TransitionDirection :Forward 0.23 :Backward 0.72 :None 0.0) _returnvalue) :!class /Script/Engine.KismetMaterialLibrary))

(fn StopTransitions)

(fn ConstructionScript
  (CallFunction|SetBackgroundFromChildActor)
  (CallFunction|AssignBackgroundMaterial (Variables|Default|GetDefaultBackgroundMaterial) true)
  (CallFunction|AssignFloorMaterial (Variables|Default|GetDefaultFloorMaterial) true)
  (CallFunction|AssignBackgroundMaterial (Variables|Default|GetDefaultBackgroundMaterial))
  (CallFunction|AssignFloorMaterial (Variables|Default|GetDefaultFloorMaterial))
  (|CallParentFunction))
