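# Top bar V-Bucks button widget for the frontend, built on the native FortMTXButton class.
# A horizontal box holds a material-instance V-Bucks icon followed by the inherited Text_Amount numeric text, styled with an outlined Burbank font and enlarged on mobile.
# The Blueprint adds no graph logic or variables, and the parent's InputActionWidget slot is left unfilled.
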WidgetBlueprint /Game/UI/Frontend/TopBar/WBP_VBucks
graphs [EventGraph {links [], nodes []}]
variables []
widgets {
 info {inheritedWidgetCount=2, parentClass=/Script/FortniteUI.FortMTXButton, rootWidgetClass=&HorizontalBox, widgetCount=4}
 widgets [
  HorizontalBox_0 {widgetClassPath=&HorizontalBox, widget=HorizontalBox_0}
  Image_Vbucks {
   widgetClassPath=&Image
   bIsVariable=true
   parent=HorizontalBox_0
   widget=Image_Vbucks
   properties {brush {
     bIsDynamicallyLoaded=false
     drawAs=Image
     imageType=FullColor
     mirroring=NoMirror
     resourceName=None
     resourceObject=/Game/UI/Frontend/TopBar/MaterialInstances/MI_UI_TopNav_VBuckIcon
     tiling=NoTile
     imageSize=<26 26>
     margin=<ltrb 0 0 0 0>
     outlineSettings {
      bUseBrushTransparency=false
      roundingType=HalfHeightRadius
      width=0
      color {colorUseRule=UseColor_Specified, specifiedColor=<rgba 0 0 0 0>}
      cornerRadii=<0 0 0 0>
     }
     tintColor {colorUseRule=UseColor_Specified, specifiedColor=<rgba 1 1 1 1>}
     uVRegion {bIsValid=false, max=<0 0>, min=<0 0>}
    }}
   slot {class=HorizontalBoxSlot, properties {horizontalAlignment=HAlign_Left, verticalAlignment=VAlign_Center, padding=<ltrb 0 0 4 0>}}
  }
  Text_Amount {
   widgetClassPath=/Script/CommonUI.CommonNumericTextBlock
   bInherited=true
   bIsVariable=true
   parent=HorizontalBox_0
   widget=Text_Amount
   properties {
    currentNumericValue=123456
    mobileFontSizeMultiplier=1.5
    text="123,456"
    font {
     bForceMonospaced=false
     bMaterialIsStencil=false
     fontMaterial=null
     fontObject=&BurbankSmall
     letterSpacing=0
     monospacedWidth=1
     size=20
     skewAmount=0
     typefaceFontName=Black
     outlineSettings {
      bApplyOutlineToDropShadows=false
      bMiteredCorners=false
      bSeparateFillAlpha=false
      outlineMaterial=/FortUILibrary/MaterialInstances/Text/MI_UI_Block_Text_Outline
      outlineSize=2
      outlineColor=<rgba 1 1 1 1>
     }
    }
    shadowOffset=<0 0>
   }
   slot {class=HorizontalBoxSlot, properties {verticalAlignment=VAlign_Center}}
  }
  InputActionWidget {widgetClassPath=&CommonActionWidget, bInherited=true, widget=null}
 ]
}
