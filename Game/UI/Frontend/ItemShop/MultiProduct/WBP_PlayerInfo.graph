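# Player info widget in the item shop's multi-product UI, showing the local player's V-Bucks, username and platform, and avatar.
# A right-aligned row stacks the V-Bucks widget above the username widget, beside the avatar in a size box.
# On Construct, the avatar is set to represent the local player.
# A mobile-game check then gates a fixed avatar box size and right alignment of the username.
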WidgetBlueprint /Game/UI/Frontend/ItemShop/MultiProduct/WBP_PlayerInfo
graphs [EventGraph {nodes [
   Event Construct()
   > RepresentLocalPlayer(self:Fort Player Avatar Object Reference=Get WBP_PlayerAvatar() -> (WBP_PlayerAvatar:WBP Player Avatar Object Reference))
   > Branch(Condition=IsMobileGame())
   n1 = Get SizeBox_Mobile() -> (SizeBox_Mobile:Size Box Object Reference)
   > SetWidthOverride(self=@n1, InWidthOverride=102)
   > SetHeightOverride(self=@n1, InHeightOverride=102)
   > SetHorizontalAlignment(
    self:Vertical Box Slot Object Reference=SlotAsVerticalBoxSlot(Widget=Get WBP_UsernameAndPlatform() -> (WBP_UsernameAndPlatform:WBP Username And Platform Object Reference))
    InHorizontalAlignment=HAlign_Right
   )
  ]}]
variables []
widgets {
 info {inheritedWidgetCount=0, parentClass=&UserWidget, rootWidgetClass=&VerticalBox, widgetCount=7}
 widgets [
  VerticalBox_0 {widgetClassPath=&VerticalBox, widget=VerticalBox_0}
  HorizontalBox_0 {
   widgetClassPath=&HorizontalBox
   parent=VerticalBox_0
   widget=HorizontalBox_0
   slot {class=VerticalBoxSlot, properties {horizontalAlignment=HAlign_Right}}
  }
  VerticalBox_1 {
   widgetClassPath=&VerticalBox
   parent=HorizontalBox_0
   widget=VerticalBox_1
   slot {class=HorizontalBoxSlot, properties {horizontalAlignment=HAlign_Right, verticalAlignment=VAlign_Center, padding=<ltrb 0 0 16 0>}}
  }
  WBP_VBucks {
   widgetClassPath=/Game/UI/Frontend/TopBar/WBP_VBucks
   bIsVariable=true
   parent=VerticalBox_1
   widget=WBP_VBucks
   properties {
    holdData=&CommonUI_HoldData_Base_C
    onButtonBaseClicked=()
    onButtonBaseDoubleClicked=()
    onButtonBaseHovered=()
    onButtonBaseUnhovered=()
    onSelectedChangedBase=()
   }
   slot {class=VerticalBoxSlot, properties {horizontalAlignment=HAlign_Right}}
  }
  WBP_UsernameAndPlatform {
   widgetClassPath=&/WBP_UsernameAndPlatform
   bIsVariable=true
   parent=VerticalBox_1
   widget=WBP_UsernameAndPlatform
   slot {class=VerticalBoxSlot}
  }
  SizeBox_Mobile {widgetClassPath=&SizeBox, bIsVariable=true, parent=HorizontalBox_0, widget=SizeBox_Mobile, slot {class=HorizontalBoxSlot}}
  WBP_PlayerAvatar {
   widgetClassPath=/Game/UI/Foundation/Profile/WBP_PlayerAvatar
   bIsVariable=true
   parent=SizeBox_Mobile
   widget=WBP_PlayerAvatar
   slot {class=SizeBoxSlot}
  }
 ]
}
